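# HUD minimap widget showing one NPC service from the collection, with icon, name and either cost or lock reason.
# PopulateData fills the name and lazy-loaded icon, then switches an inner panel on the locked flag; PopulateEmpty shows a blank overlay.
# Locked services show their lock text, falling back to the required Rep level.
# Unlocked ones show resource icon and FormatCost's amount or range, hidden without a valid resource and positive minimum.
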WidgetBlueprint /FortniteCoreUI/HUD/MiniMap/Services/CollectionNPCServiceWidget
graphs [
 EventGraph {nodes [
   n1 = Event PopulateData!() -> (OutputDelegate, then, WidgetData:"NPCService Widget Data Structure (by ref)", bShowLine:bool) {is=title K2Node_Event}
   > SetActiveWidgetIndex(self=Get ContentSwitcher() -> (ContentSwitcher:Widget Switcher Object Reference), Index=0)
   > n2 = Sequence!(execute) -> (then_0, then_1, then_2:Exec)
   n3 = Break NPCService Widget Data!(NPCServiceWidgetData:"NPCService Widget Data Structure (by ref)"=@n1.WidgetData) -> (
    CostResource:Fort Resource Item Definition Object Reference
    CostAmountMin:int
    CostAmountMax:int
    bIsLocked:bool=~
    MinRequiredBenefitLevel:int
    LockText:Text
   ) {is=title K2Node_BreakStruct}
   SetActiveWidgetIndex(
    execute=@n2.then_1
    self=Get InnerSwitcher() -> (InnerSwitcher:Widget Switcher Object Reference)
    Index=Select(Option 0:int=0, Option 1:int=1, Index=@n3.bIsLocked) -> (ReturnValue:int)
   )
   > n4 = Branch(Condition=@n3.bIsLocked)
   n5 = TextIsEmpty(InText=@n3.LockText)
   > "SetText (Text)"(
    self=Get Text_LockReason() -> (Text_LockReason:Common Text Object Reference)
    InText=Format Text!(
     Format=Select(Option 0:Text=@n3.LockText, Option 1:Text="LOCTEXT(\"Rep {0}\")", Index=@n5) -> (ReturnValue:Text)
     0:int=@n3.MinRequiredBenefitLevel
    ) -> (Result)
   )
   n6 = Branch(execute=@n4.else, Condition=AND Boolean(A=IsValid(Object=@n3), B="integer > integer"(A=@n3.CostAmountMin)))
   n7 = Get CostPanel() -> (CostPanel:Horizontal Box Object Reference)
   > SetVisibility(self=@n7, InVisibility=HitTestInvisible)
   > "SetText (Text)"(
    self=Get Text_Cost() -> (Text_Cost:Common Text Object Reference)
    InText=FormatCost!(MinCost:int=@n3.CostAmountMin, MaxCost:int=@n3.CostAmountMax) -> (CostText:Text)
   )
   > SetBrushFromSoftTexture!(
    execute
    self:Image Object Reference=Get LazyImage_Cost() -> (LazyImage_Cost:Fort Lazy Image Object Reference)
    SoftTexture:Texture 2D Soft Object Reference=GetSmallPreviewImage!(self:Fort Item Definition Object Reference=@n3) -> (ReturnValue:Texture 2D Soft Object Reference)
    bMatchSize:bool=false
   ) -> (then)
   SetVisibility(execute=@n6.else, self=@n7, InVisibility=Collapsed)
   n8 = Break NPCService Data!(NPCServiceData:"NPCService Data Structure (by ref)"=Break NPCService Widget Data!(NPCServiceWidgetData=@n1.WidgetData) -> (ServiceData:NPCService Data Structure, CostAmountMin, CostAmountMax, LockText)) -> (Name:Text, Icon:Soft Object Reference) {is=title K2Node_BreakStruct}
   "SetText (Text)"(execute=@n2.then_0, self=Get Text_ServiceName() -> (Text_ServiceName:Common Text Object Reference), InText=@n8)
   n9 = Get LazyImage_Icon() -> (LazyImage_Icon:Fort Lazy Image Object Reference)
   > SetBrushFromLazyDisplayAsset!(execute, self:Common Lazy Image Object Reference=@n9, LazyObject:"Soft Object Reference (by ref)"=@n8.Icon, bMatchTextureSize:bool=false) -> (then)
   > SetVisibility(
    self=@n9
    InVisibility=Select(
     Option 0:ESlateVisibility Enum=HitTestInvisible
     Option 1:ESlateVisibility Enum=Collapsed
     Index=IsAssetNull!(Asset:Soft Object Reference=@n8.Icon) -> (ReturnValue:bool)
    ) -> (ReturnValue:ESlateVisibility Enum)
   )
   Event PopulateEmpty() {is=title K2Node_Event}
   > SetActiveWidgetIndex(self=Get ContentSwitcher(), Index=1)
   Event Construct()
   > SetWrapTextWidth!(
    execute
    self:Common Text Object Reference=Get Text_ServiceName()
    InWrapTextAt:int=Select(Option 0:int=0, Option 1:int=150, Index=IsMobileGame()) -> (ReturnValue:int)
   ) -> (then)
  ]}
 FormatCost {nodes [
   n1 = FormatCost!() -> (then, MinCost:int, MaxCost:int) {is=title K2Node_FunctionEntry}
   > n2 = Branch(Condition="Equal (Integer)"(A=@n1, B=@n1.MaxCost))
   > Return Node!(execute, CostText:Text="To Text (Integer)"(Value=@n1))
   Return Node(execute=@n2.else, CostText=Format Text!(Format="LOCTEXT(\"{0} - {1}\")", 0=@n1, 1:int=@n1.MaxCost) -> (Result))
  ]}
]
variables []
widgets {
 info {inheritedWidgetCount=0, parentClass=/Script/FortniteUI.CollectionNPCServiceWidget, rootWidgetClass=&SizeBox, widgetCount=19}
 widgets [
  SizeBox_0 {widgetClassPath=&SizeBox, widget=SizeBox_0, properties {bOverride_WidthOverride=true, widthOverride=300}}
  ContentSwitcher {widgetClassPath=&WidgetSwitcher, bIsVariable=true, parent=SizeBox_0, widget=ContentSwitcher, slot {class=SizeBoxSlot}}
  Filled {widgetClassPath=&Overlay, parent=ContentSwitcher, widget=Filled, slot {class=WidgetSwitcherSlot}}
  HorizontalBox_39 {
   widgetClassPath=&HorizontalBox
   parent=Filled
   widget=HorizontalBox_39
   slot {class=OverlaySlot, properties {horizontalAlignment=HAlign_Fill, verticalAlignment=VAlign_Fill, padding=<ltrb 6 6 6 6>}}
  }
  Overlay_0 {
   widgetClassPath=&Overlay
   parent=HorizontalBox_39
   widget=Overlay_0
   slot {class=HorizontalBoxSlot, properties {horizontalAlignment=HAlign_Left}}
  }
  LazyImage_IconBg {
   widgetClassPath=&Image
   bIsVariable=true
   parent=Overlay_0
   widget=LazyImage_IconBg
   properties {brush {
     bIsDynamicallyLoaded=false
     drawAs=Image
     imageType=NoImage
     mirroring=NoMirror
     resourceName=None
     resourceObject=/FortniteCoreUI/HUD/MiniMap/Services/Textures/MI_UI_NPCCollect_IconBg
     tiling=NoTile
     imageSize=<70 70>
     margin=<ltrb 0 0 0 0>
     outlineSettings {
      bUseBrushTransparency=false
      roundingType=HalfHeightRadius
      width=0
      color {colorUseRule=UseColor_Specified, specifiedColor=<rgba 0 0 0 0>}
      cornerRadii=<0 0 0 0>
     }
     tintColor {colorUseRule=UseColor_Specified, specifiedColor=<rgba 1 1 1 1>}
     uVRegion {bIsValid=false, max=<0 0>, min=<0 0>}
    }}
   slot {class=OverlaySlot, properties {horizontalAlignment=HAlign_Fill, verticalAlignment=VAlign_Fill}}
  }
  LazyImage_Icon {
   widgetClassPath=&FortLazyImage
   bIsVariable=true
   parent=Overlay_0
   widget=LazyImage_Icon
   properties {brush {
     bIsDynamicallyLoaded=false
     drawAs=Image
     imageType=NoImage
     mirroring=NoMirror
     resourceName=None
     resourceObject=null
     tiling=NoTile
     imageSize=<64 64>
     margin=<ltrb 0 0 0 0>
     outlineSettings {
      bUseBrushTransparency=false
      roundingType=HalfHeightRadius
      width=0
      color {colorUseRule=UseColor_Specified, specifiedColor=<rgba 0 0 0 0>}
      cornerRadii=<0 0 0 0>
     }
     tintColor {colorUseRule=UseColor_Specified, specifiedColor=<rgba 1 1 1 1>}
     uVRegion {bIsValid=false, max=<0 0>, min=<0 0>}
    }}
   slot {class=OverlaySlot, properties {horizontalAlignment=HAlign_Center, verticalAlignment=VAlign_Center}}
  }
  VerticalBox_0 {
   widgetClassPath=&VerticalBox
   parent=HorizontalBox_39
   widget=VerticalBox_0
   slot {class=HorizontalBoxSlot, properties {verticalAlignment=VAlign_Top, padding=<ltrb 8 0 0 0>, size {sizeRule=Fill, value=1}}}
  }
  Text_ServiceName {
   widgetClassPath=&CommonTextBlock
   bIsVariable=true
   parent=VerticalBox_0
   widget=Text_ServiceName
   properties {
    mobileFontSizeMultiplier=1.5
    scrollStyle=&TextScrollStyle-Base_C
    text=Service Name
    textTransformPolicy=ToUpper
    font {
     bForceMonospaced=false
     bMaterialIsStencil=false
     fontMaterial=null
     fontObject=&HeadingNow
     letterSpacing=0
     monospacedWidth=1
     size=12
     skewAmount=0
     typefaceFontName=76 Bold
     outlineSettings {
      bApplyOutlineToDropShadows=false
      bMiteredCorners=false
      bSeparateFillAlpha=false
      outlineMaterial=null
      outlineSize=0
      outlineColor=<rgba 0 0 0 1>
     }
    }
    shadowOffset=<0 0>
   }
   slot {class=VerticalBoxSlot, properties {padding=<ltrb 0 0 0 3>}}
  }
  InnerSwitcher {
   widgetClassPath=&WidgetSwitcher
   bIsVariable=true
   parent=VerticalBox_0
   widget=InnerSwitcher
   slot {class=VerticalBoxSlot, properties {verticalAlignment=VAlign_Bottom, padding=<ltrb 0 6 0 0>}}
  }
  CostPanel {widgetClassPath=&HorizontalBox, bIsVariable=true, parent=InnerSwitcher, widget=CostPanel, slot {class=WidgetSwitcherSlot}}
  LazyImage_Cost {
   widgetClassPath=&FortLazyImage
   bIsVariable=true
   parent=CostPanel
   widget=LazyImage_Cost
   slot {class=HorizontalBoxSlot, properties {verticalAlignment=VAlign_Center}}
  }
  Text_Cost {
   widgetClassPath=&CommonTextBlock
   bIsVariable=true
   parent=CostPanel
   widget=Text_Cost
   properties {
    lineHeightPercentage=1.28
    mobileFontSizeMultiplier=1.5
    text=9999
    font {
     bForceMonospaced=false
     bMaterialIsStencil=false
     fontMaterial=null
     fontObject=&HeadingNow
     letterSpacing=0
     monospacedWidth=1
     size=12.75
     skewAmount=0
     typefaceFontName=75 Medium
     outlineSettings {
      bApplyOutlineToDropShadows=false
      bMiteredCorners=false
      bSeparateFillAlpha=false
      outlineMaterial=null
      outlineSize=0
      outlineColor=<rgba 0 0 0 1>
     }
    }
    shadowOffset=<0 0>
    strikeBrush {
     bIsDynamicallyLoaded=false
     drawAs=NoDrawType
     imageType=NoImage
     mirroring=NoMirror
     resourceName=None
     resourceObject=null
     tiling=NoTile
     imageSize=<32 32>
     margin=<ltrb 0 0 0 0>
     outlineSettings {
      bUseBrushTransparency=false
      roundingType=HalfHeightRadius
      width=0
      color {colorUseRule=UseColor_Specified, specifiedColor=<rgba 0 0 0 0>}
      cornerRadii=<0 0 0 0>
     }
     tintColor {colorUseRule=UseColor_Specified, specifiedColor=<rgba 1 1 1 1>}
     uVRegion {bIsValid=false, max=<0 0>, min=<0 0>}
    }
   }
   slot {class=HorizontalBoxSlot, properties {verticalAlignment=VAlign_Center, padding=<ltrb 4 0 0 0>}}
  }
  LockedPanel {widgetClassPath=&HorizontalBox, bIsVariable=true, parent=InnerSwitcher, widget=LockedPanel, slot {class=WidgetSwitcherSlot}}
  LockStuff {
   widgetClassPath=&Overlay
   bIsVariable=true
   parent=LockedPanel
   widget=LockStuff
   properties {renderTransformPivot=<0 1>}
   slot {class=HorizontalBoxSlot, properties {horizontalAlignment=HAlign_Center, verticalAlignment=VAlign_Center}}
  }
  Image_305 {
   widgetClassPath=&Image
   bIsVariable=true
   parent=LockStuff
   widget=Image_305
   properties {
    brush {
     bIsDynamicallyLoaded=false
     drawAs=Image
     imageType=NoImage
     mirroring=NoMirror
     resourceName=None
     resourceObject=null
     tiling=NoTile
     imageSize=<22 22>
     margin=<ltrb 0 0 0 0>
     outlineSettings {
      bUseBrushTransparency=false
      roundingType=HalfHeightRadius
      width=0
      color {colorUseRule=UseColor_Specified, specifiedColor=<rgba 0 0 0 0>}
      cornerRadii=<0 0 0 0>
     }
     tintColor {colorUseRule=UseColor_Specified, specifiedColor=<rgba 1 1 1 1>}
     uVRegion {bIsValid=false, max=<0 0>, min=<0 0>}
    }
    colorAndOpacity=<rgba 1 0 0.2988 1>
    renderTransform {angle=0, scale=<1 1>, shear=<-9 0>, translation=<0 0>}
   }
   slot {class=OverlaySlot, properties {horizontalAlignment=HAlign_Fill, verticalAlignment=VAlign_Fill}}
  }
  ImgLockIcon {
   widgetClassPath=&Image
   bIsVariable=true
   parent=LockStuff
   widget=ImgLockIcon
   properties {brush {
     bIsDynamicallyLoaded=false
     drawAs=Image
     imageType=NoImage
     mirroring=NoMirror
     resourceName=None
     resourceObject=/Game/UI/Foundation/Textures/Icons/Locks/T-Icon-Lock-64
     tiling=NoTile
     imageSize=<16 16>
     margin=<ltrb 0 0 0 0>
     outlineSettings {
      bUseBrushTransparency=false
      roundingType=HalfHeightRadius
      width=0
      color {colorUseRule=UseColor_Specified, specifiedColor=<rgba 0 0 0 0>}
      cornerRadii=<0 0 0 0>
     }
     tintColor {colorUseRule=UseColor_Specified, specifiedColor=<rgba 1 1 1 1>}
     uVRegion {bIsValid=false, max=<0 0>, min=<0 0>}
    }}
   slot {class=OverlaySlot, properties {horizontalAlignment=HAlign_Center, verticalAlignment=VAlign_Center}}
  }
  Text_LockReason {
   widgetClassPath=&CommonTextBlock
   bIsVariable=true
   parent=LockedPanel
   widget=Text_LockReason
   properties {
    autoWrapText=true
    mobileFontSizeMultiplier=1.5
    text=Rep 8
    colorAndOpacity {colorUseRule=UseColor_Specified, specifiedColor=<rgba 1 0 0.3005 1>}
    font {
     bForceMonospaced=false
     bMaterialIsStencil=false
     fontMaterial=null
     fontObject=&BurbankSmall
     letterSpacing=0
     monospacedWidth=1
     size=10.5
     skewAmount=0.14
     typefaceFontName=Bold
     outlineSettings {
      bApplyOutlineToDropShadows=false
      bMiteredCorners=false
      bSeparateFillAlpha=false
      outlineMaterial=null
      outlineSize=0
      outlineColor=<rgba 0 0 0 1>
     }
    }
    shadowOffset=<0 0>
    strikeBrush {
     bIsDynamicallyLoaded=false
     drawAs=NoDrawType
     imageType=NoImage
     mirroring=NoMirror
     resourceName=None
     resourceObject=null
     tiling=NoTile
     imageSize=<32 32>
     margin=<ltrb 0 0 0 0>
     outlineSettings {
      bUseBrushTransparency=false
      roundingType=HalfHeightRadius
      width=0
      color {colorUseRule=UseColor_Specified, specifiedColor=<rgba 0 0 0 0>}
      cornerRadii=<0 0 0 0>
     }
     tintColor {colorUseRule=UseColor_Specified, specifiedColor=<rgba 1 1 1 1>}
     uVRegion {bIsValid=false, max=<0 0>, min=<0 0>}
    }
   }
   slot {class=HorizontalBoxSlot, properties {verticalAlignment=VAlign_Center, padding=<ltrb 4 0 4 0>, size {sizeRule=Fill, value=1}}}
  }
  Overlay_83 {widgetClassPath=&Overlay, parent=ContentSwitcher, widget=Overlay_83, slot {class=WidgetSwitcherSlot}}
 ]
}
