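# Widget Blueprint for a group of collection NPC service entries, kept under the HUD minimap services.
# Its parent is the native CollectionNPCServiceGroup class, and its event graph holds only unconnected stock events.
# An Overlay root holds the inherited EntryBox_Services dynamic entry box, which creates CollectionNPCServiceWidget entries.
# Entries share the space through Fill sizing, with horizontal spacing between them.
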WidgetBlueprint /FortniteCoreUI/HUD/MiniMap/Services/CollectionNPCServiceGroup
graphs [EventGraph {nodes [
   Event PreConstruct()
   Event Construct()
   Event Tick()
  ]}]
variables []
widgets {
 info {inheritedWidgetCount=1, parentClass=/Script/FortniteUI.CollectionNPCServiceGroup, rootWidgetClass=&Overlay, widgetCount=2}
 widgets [
  Overlay_14 {widgetClassPath=&Overlay, widget=Overlay_14}
  EntryBox_Services {
   widgetClassPath=&DynamicEntryBox
   bInherited=true
   bIsVariable=true
   parent=Overlay_14
   widget=EntryBox_Services
   properties {
    entryWidgetClass=&/CollectionNPCServiceWidget.CollectionNPCServiceWidget_C
    numDesignerPreviewEntries=2
    entrySizeRule {sizeRule=Fill, value=1}
    entrySpacing=<8 0>
   }
   slot {class=OverlaySlot, properties {horizontalAlignment=HAlign_Fill}}
  }
 ]
}
